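# Physics asset for the RockCrest gadget, a prop used by an emote in CosmeticCompanions.
# It has a single body on the main bone, shaped as one upright-offset capsule that roughly encloses the prop.
# There are no constraints, so it provides simple collision and bounds rather than articulated ragdoll simulation.
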PhysicsAsset /CosmeticCompanions/Assets/Props/Emotes/RockCrest/Mesh/PA_RockCrest_Gadget
bodies {main [{shapeType=Capsule, length=41.1789, radius=35.3609, center=<0.509 0.7241 34.1981>, rotation=<rot 0.7412 -90 -0.0158>}]}
constraints []
